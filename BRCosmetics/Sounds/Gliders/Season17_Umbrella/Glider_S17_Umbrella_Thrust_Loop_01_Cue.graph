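# Sound cue for the looping thrust audio of the Season 17 Umbrella glider.
# A looping wave feeds a continuous modulator whose AnyDirection parameter drives both pitch and volume, so the loop can rise in pitch and loudness at runtime.
# The mixer and root modulator above it add no variation.
# Glider-specific attenuation, sound class and cosmetic loop properties sheet control playback.
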SoundCue /BRCosmetics/Sounds/Gliders/Season17_Umbrella/Glider_S17_Umbrella_Thrust_Loop_01_Cue
root=Modulator_0
nodes [
 {id=Mixer_0, children [ModulatorContinuous_0], properties {inputVolume [1]}}
 {
  id=ModulatorContinuous_0
  children [WavePlayer_0]
  properties {
   pitchModulationParams {default=0, maxOutput=1.2, minOutput=1, parameterName=AnyDirection}
   volumeModulationParams {default=0, minOutput=0.3, parameterName=AnyDirection}
  }
 }
 {id=Modulator_0, children [Mixer_0], properties {pitchMax=1, pitchMin=1, volumeMax=1, volumeMin=1}}
 {
  id=WavePlayer_0
  properties {bLooping=true, soundWaveAssetPtr=&/Glider_S17_Umbrella_Thrust_Loop_01}
 }
]
properties {
 attenuationSettings=&Fort_Player_Glider_Thrust_Athena_Att
 audioPropertiesSheet=&APS_Cosmetic_Glider_Loop
 maxDistance=12500
 soundClassObject=&Fort_Gliders_Land
 volumeMultiplier=0.35
}
